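# Example 04 of the MaterialGym tech-art patterns: a Default Lit material with an animated emissive triangle pattern.
# MF_TextureUVs turns a triangle UV mask texture into texture-based UVs alongside base UVs, exposed through named reroutes.
# Time plus blurred noise drives a remapped, exponent-shaped sine between DF_Min and DF_Max.
# A smoothstepped triangle distance field is subtracted in the mask blending Color to EmissiveColor, alongside a tiled, flatness-adjusted triangle normal map.
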Material /Game/Global/Materials/MaterialGym/Materials/TechArtPatterns/Examples/Example_04/M_Example_04
graph {
 properties {bUsedWithMeshDeformer=true, bUsedWithSkeletalMesh=true, pixelDepthOffsetMode=PDOM_Legacy, usedShadingModels=Default Lit}
 nodes [
  n1 = NamedRerouteUsage()
  n2 = SmoothStep(Value=TextureSample(UVs=NamedRerouteUsage()) {samplerType=SAMPLERTYPE_Alpha, texture=/Game/Global/Materials/MaterialGym/Materials/TechArtPatterns/Textures/T_Triangles_DF}.R) {constMax=0.2, constMin=0.1}
  LinearInterpolate(
   A=VectorParameter() {parameterName=Color, defaultValue=<rgba 0 0 0 1>}.RGB
   B=VectorParameter() {parameterName=EmissiveColor, defaultValue=<rgba 1 0.2219 0 1>}.RGB
   Alpha=SmoothStep(Value=Subtract(A=@n1, B=@n2)) {constMax=0.5}
  )
  n3 = ScalarParameter() {defaultValue=3, parameterName=PatternTiling}
  n4 = Multiply(A=@n3, B=TextureCoordinate())
  NamedRerouteDeclaration(None=MaterialFunctionCall(
    Normal=TextureSample(UVs=@n4) {samplerType=SAMPLERTYPE_Normal, texture=/Game/Global/Materials/MaterialGym/Materials/TechArtPatterns/Textures/T_Triangles_N}.RGB
    Flatness=ScalarParameter() {defaultValue=-1, parameterName=NormalFlatness}
   ) {materialFunction=&FlattenNormal}.Result) {name=Normal_OUT, nodeColor=<rgba 0 1 0.3438 1>}
  n5 = MaterialFunctionCall!(
   Texture=TextureObject() {samplerType=SAMPLERTYPE_Masks, texture=/Game/Global/Materials/MaterialGym/Materials/TechArtPatterns/Textures/T_Triangles_UV}
   TilingXY=@n3
   OffsetXY
   AnimateOffset
   UseParallax
   Height
  ) {materialFunction=/Game/Global/Materials/MaterialGym/Materials/TechArtPatterns/MaterialFunctions/MF_TextureUVs}
  NamedRerouteDeclaration(None=@n5.BaseUVs) {name=BaseUVs, nodeColor=<rgba 0.9922 1 0 1>}
  n6 = NamedRerouteUsage()
  n7 = Add(
   A=Multiply(A=Time(), B=ScalarParameter() {defaultValue=0.2, parameterName=NoiseSpeed}) {constB=0.1}
   B=Multiply(
    A=TextureSample(UVs=@n6) {samplerType=SAMPLERTYPE_Grayscale, texture=&LowResBlurredNoise}.R
    B=ScalarParameter() {defaultValue=3, parameterName=NoiseMultiply}
   )
  )
  n8 = Saturate(None=ConstantBiasScale(None=Sine(None=@n7)))
  n9 = ScalarParameter() {defaultValue=2, parameterName=NoiseExponent}
  NamedRerouteDeclaration(None=LinearInterpolate(
    A=ScalarParameter() {defaultValue=0.05, parameterName=DF_Min}
    B=ScalarParameter() {defaultValue=1, parameterName=DF_Max}
    Alpha=Power(Base=@n8, Exp=@n9) {constExponent=0.5}
   )) {name=Noise, nodeColor=<rgba 0.8516 1 0 1>}
  NamedRerouteDeclaration(None=@n5.TextureUVs) {name=TextureBasedUV, desc=Texture based UV, nodeColor=<rgba 0 0.4844 1 1>}
  NamedRerouteUsage()
 ]
}
